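; Parent Blueprint for Creative grass floor tiles that lets user options recolor the ground and grass.
; On begin play and option load or change, it runs cosmetic-only: caches default materials, MIDs and colors, then selects custom or default colors.
; It writes those to tile and grass MIDs, plus any differing material the phone tool applied.
; Grass functions take mesh and MID as inputs to support variants with multiple grass meshes.
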(event OnPropertiesLoaded(ToyOptionsComponent)
  (Networking|CanExecuteCosmeticEvents
    (:True
      (CallFunction|CacheAllDefaultMaterials true)
      (CallFunction|UpdateVisualsonClient true))))

(event Collision|EventActorBeginOverlap
  (params (OtherActor :type Actor%20Object%20Reference)))

(event OnAnyPropertyChanged(ToyOptionsComponent)
  (Networking|CanExecuteCosmeticEvents
    (:True
      (CallFunction|CacheAllDefaultMaterials true)
      (CallFunction|UpdateVisualsonClient true))))

(event EventBeginPlay
  (|CallParentFunction)
  (Networking|CanExecuteCosmeticEvents
    (:True
      (CallFunction|CacheAllDefaultMaterials true)
      (CallFunction|UpdateVisualsonClient true))))

(event EventTick
  (params (DeltaSeconds :type Float%20%28single-precision%29)))

(fn CacheAllDefaultMaterials
  (params (CacheTileMaterial :type Boolean :default true) (CacheGrassMaterial :type Boolean :default true))
  (if CacheTileMaterial
    ;[1 1 1] Cache Tile
    (bind _newlycached (Default|CacheDefaultMaterialandMID (Variables|Debug|GetStaticMeshComponent) (Variables|Materials|GetTileMatIndex) (Variables|Materials|GetDefaultTileMaterial) (Variables|Materials|MID|GetMID_Tile)))
    (if _newlycached
      (Default|CacheDefaultColor (Variables|Materials|MID|GetMID_Tile) (Variables|Parameters|GetGroundColor_Param) (Variables|Parameters|GetGroundColor_Default)) ; Default ground color
      (Default|CacheDefaultColor (Variables|Materials|MID|GetMID_Tile) (Variables|Parameters|GetGrassColor_Tile_Param) (Variables|Parameters|GetGrassColor_Tile_Default)) ; Default grass color (tile)
      (else
        _))
    (if (Variables|GetCacheGrassMaterial)
      (Default|CacheGrassMaterial (Variables|Default|GetGrass) (Variables|Materials|GetGrassMatIndex) (Variables|Materials|GetDefaultGrassMaterial) (Variables|Materials|MID|GetMID_Grass)) ; The input is all split so CP_ColorGrass_Solid_3Quarter works,\nsince it has three grass meshes.
    )
    (else
      _)))

(fn CacheDefaultMaterialAndMID
  (params (Target%20Mesh :type Static%20Mesh%20Component%20Object%20Reference) (Target%20Material%20Index :type Integer) (Material%20Variable :type Material%20Interface%20Object%20Reference%20%28by%20ref%29) (MID%20Variable :type Material%20Instance%20Dynamic%20Object%20Reference%20%28by%20ref%29))
  (bind _target_material_index (Variables|GetTarget%20Material%20Index))
  (bind _target_mesh (Variables|GetTarget%20Mesh))
  (bind _returnvalue (Rendering|Material|GetMaterial _target_mesh _target_material_index))
  (bind _material_variable (Variables|GetMaterial%20Variable))
  (bind _target_material_index_2 (Variables|GetTarget%20Material%20Index))
  (bind _target_mesh_2 (Variables|GetTarget%20Mesh))
  (bind _mid (Material|GetMIDfromMaterial _target_mesh_2 _target_material_index_2 _material_variable))
  ;[1 0.7813 0.01256 0.5] Check and cache default material
  (Utilities|IsValid (Variables|GetMaterial%20Variable)
    (:Is%20Valid
      ;[0.7537 1 0.06522 0.5] Set the MID if needed
      (Utilities|IsValid (Variables|GetMID%20Variable)
        (:Is%20Valid
          _)
        (:Is%20Not%20Valid
          (Variables|SetBy-RefVar (Variables|GetMID%20Variable) _mid))) ; Is the MID valid? If not, re-set it.
      (return false))
    (:Is%20Not%20Valid
      (Utilities|IsValid _returnvalue
        (:Is%20Valid
          (Variables|SetBy-RefVar (Variables|GetMaterial%20Variable) _returnvalue)
          (Variables|SetBy-RefVar (Variables|GetMID%20Variable) _mid)
          (return true))
        (:Is%20Not%20Valid
          (Development|PrintString (Utilities|String|Append "ColorGrass CacheDefaultMaterial FAILED to cache: " (Utilities|GetDisplayName (Variables|GetTarget%20Mesh)) " (" (Utilities|GetDisplayName (Components|GetOwner (Variables|GetTarget%20Mesh))) ")" " , Element Index " (Utilities|String|ToString(Integer) (Variables|GetTarget%20Material%20Index))) false)
          (return false))))) ; Is this material cached yet?
)

(fn CacheGrassMaterial
  (params (Grass%20Mesh :type Static%20Mesh%20Component%20Object%20Reference%20%28by%20ref%29) (Grass%20Material%20Index :type Integer) (Default%20Grass%20Mat%20Variable :type Material%20Interface%20Object%20Reference%20%28by%20ref%29) (Grass%20MID%20Variable :type Material%20Instance%20Dynamic%20Object%20Reference%20%28by%20ref%29) (CacheDefaultColor :type Boolean :default true))
  ;[1 1 1] Cache Grass
  (bind _newlycached (Default|CacheDefaultMaterialandMID Grass%20Mesh Grass%20Material%20Index Default%20Grass%20Mat%20Variable Grass%20MID%20Variable))
  (if (and _newlycached (Variables|GetCacheDefaultColor))
    (Default|CacheDefaultColor (Variables|GetGrass%20MID%20Variable) (Variables|Parameters|GetGrassColor_GrassMesh_Param) (Variables|Parameters|GetGrassColor_GrassMesh_Default)) ; Default grass color (grass mesh)
  ))

(fn SetCurrentColors
  (params (SetCurrentTileColors :type Boolean) (SetCurrentGrassColor :type Boolean))
  (if SetCurrentTileColors
    (Variables|Parameters|SetGroundColor_Current (Math|Color|SelectColor (Variables|UserOptions|Basic|GetGroundColor) (Variables|Parameters|GetGroundColor_Default) (Variables|UserOptions|Basic|GetCustomizeGroundColor)))
    (Variables|Parameters|SetGrassColor_Tile_Current (Math|Color|SelectColor (Variables|UserOptions|Basic|GetGrassColor) (Variables|Parameters|GetGrassColor_Tile_Default) (Variables|UserOptions|Basic|GetCustomizeGrassColor)))
    (else
      _))
  (if SetCurrentGrassColor
    (Variables|Parameters|SetGrassColor_GrassMesh_Current (Math|Color|SelectColor (Variables|UserOptions|Basic|GetGrassColor) (Variables|Parameters|GetGrassColor_GrassMesh_Default) (Variables|UserOptions|Basic|GetCustomizeGrassColor)))))

(fn SetGrassMat
  (params (GrassMesh :type Static%20Mesh%20Component%20Object%20Reference%20%28by%20ref%29) (GrassMID :type Material%20Instance%20Dynamic%20Object%20Reference%20%28by%20ref%29))
  (Utilities|Array|Add (LocalVariables|GetMIDsToChange) (Variables|GetGrassMID))
  (if (!= (Rendering|Material|GetMaterial (Variables|GetGrassMesh) (Variables|Materials|GetGrassMatIndex)) (Variables|GetGrassMID))
    ;[1 1 1] Ensure we also change the material the mesh currently has, which was added by the phone tool
    (Utilities|Array|Add (LocalVariables|GetMIDsToChange) (Material|GetMIDfromMaterial (Variables|GetGrassMesh) (Variables|Materials|GetGrassMatIndex)))
    (else
      _))
  ;[1 1 1] Grass Color
  (Rendering|Material|SetVectorParameterValue (LocalVariables|GetMIDsToChange) (Variables|Parameters|GetGrassColor_GrassMesh_Param) (Variables|Parameters|GetGrassColor_GrassMesh_Current) :!class /Script/Engine.MaterialInstanceDynamic))

(fn SetTileMat
  (Utilities|Array|Add (LocalVariables|GetMIDsToChange) (Variables|Materials|MID|GetMID_Tile))
  (if (!= (Rendering|Material|GetMaterial (Variables|Debug|GetStaticMeshComponent) (Variables|Materials|GetTileMatIndex)) (Variables|Materials|MID|GetMID_Tile))
    ;[1 1 1] Ensure we also change the material the mesh currently has, which was added by the phone tool
    (Utilities|Array|Add (LocalVariables|GetMIDsToChange) (Material|GetMIDfromMaterial (Variables|Debug|GetStaticMeshComponent) (Variables|Materials|GetTileMatIndex)))
    (else
      _))
  ;[1 1 1] Ground Color
  (Rendering|Material|SetVectorParameterValue (LocalVariables|GetMIDsToChange) (Variables|Parameters|GetGroundColor_Param) (Variables|Parameters|GetGroundColor_Current) :!class /Script/Engine.MaterialInstanceDynamic)
  ;[1 1 1] Grass Color
  (Rendering|Material|SetVectorParameterValue (LocalVariables|GetMIDsToChange) (Variables|Parameters|GetGrassColor_Tile_Param) (Variables|Parameters|GetGrassColor_Tile_Current) :!class /Script/Engine.MaterialInstanceDynamic))

(fn UpdateVisualsOnClient
  (params (UpdateTileMesh :type Boolean :default true) (UpdateGrassMeshes :type Boolean :default true))
  (CallFunction|SetCurrentColors (Variables|GetUpdateTileMesh) (Variables|GetUpdateGrassMeshes))
  (if (Variables|GetUpdateTileMesh)
    (Default|SetTileMat)
    (else
      _))
  (if (Variables|GetUpdateGrassMeshes)
    (Default|SetGrassMat (Variables|Default|GetGrass) (Variables|Materials|MID|GetMID_Grass)) ; The input is all split so CP_ColorGrass_Solid_3Quarter works,\nsince it has three grass meshes.
  ))

(fn ConstructionScript
  (|CallParentFunction))
