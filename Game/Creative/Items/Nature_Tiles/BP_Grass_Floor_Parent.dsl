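; Parent Blueprint for grass floor tiles in Creative, with a static mesh and a Grass component.
; On BeginPlay it caches each component's current collision-enabled state in the DefaultCollisions map, categorized under TrickTile.
; GetDefaultCollisions returns that map and GetStaticMeshComponents returns its keys, so both are populated only once play begins.
; The construction script is empty.
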(event EventBeginPlay
  (Variables|TrickTile|SetDefaultCollisions (Utilities|Map|MakeMap (Variables|Debug|GetStaticMeshComponent) (Collision|GetCollisionEnabled (Variables|Debug|GetStaticMeshComponent)) (Variables|Default|GetGrass) (Collision|GetCollisionEnabled (Variables|Default|GetGrass)))))

(fn GetDefaultCollisions
  (return (Variables|TrickTile|GetDefaultCollisions)))

(fn GetStaticMeshComponents
  (bind _keys (Utilities|Map|Keys (Variables|TrickTile|GetDefaultCollisions)))
  (return _keys))

(fn ConstructionScript)
