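; Creative nature-tile prop, a three-quarter solid grass piece, that extends its parent's color-grass material handling to three grass meshes.
; CacheGrassMaterial and SetGrassMat overrides call the parent once for the passed-in mesh, then again for two inverted-grass components with their own material variables.
; All three meshes share one material index, and the two extra meshes never cache the default color.
; The construction script is empty.
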(fn CacheGrassMaterial
  (params (Grass%20Mesh :type Static%20Mesh%20Component%20Object%20Reference%20%28by%20ref%29) (Grass%20Material%20Index :type Integer) (Default%20Grass%20Mat%20Variable :type Material%20Interface%20Object%20Reference%20%28by%20ref%29) (Grass%20MID%20Variable :type Material%20Instance%20Dynamic%20Object%20Reference%20%28by%20ref%29) (CacheDefaultColor :type Boolean))
  (|CallParentFunction Grass%20Mesh Grass%20Material%20Index Default%20Grass%20Mat%20Variable Grass%20MID%20Variable CacheDefaultColor) ; First grass mesh
  (|CallParentFunction (Variables|Default|GetCreative_SM_IslandGrass_InvertedGrass_131) (Variables|GetGrass%20Material%20Index) (Variables|Default|GetDefaultGrassMaterial_2) (Variables|Default|GetMID_Grass_2) false) ; Second grass mesh
  (|CallParentFunction (Variables|Default|GetCreative_SM_IslandGrass_InvertedGrass_13) (Variables|GetGrass%20Material%20Index) (Variables|Default|GetDefaultGrassMaterial_3) (Variables|Default|GetMID_Grass_3) false) ; Third grass mesh
  (return))

(fn SetGrassMat
  (params (GrassMesh :type Static%20Mesh%20Component%20Object%20Reference%20%28by%20ref%29) (GrassMID :type Material%20Instance%20Dynamic%20Object%20Reference%20%28by%20ref%29))
  (|CallParentFunction GrassMesh GrassMID) ; First grass mesh
  (|CallParentFunction (Variables|Default|GetCreative_SM_IslandGrass_InvertedGrass_131) (Variables|Default|GetMID_Grass_2)) ; Second grass mesh
  (|CallParentFunction (Variables|Default|GetCreative_SM_IslandGrass_InvertedGrass_13) (Variables|Default|GetMID_Grass_3)) ; Third grass mesh
)

(fn ConstructionScript)
