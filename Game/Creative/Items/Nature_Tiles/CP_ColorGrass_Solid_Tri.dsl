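; Creative nature tile Blueprint, named as a solid triangular colored-grass piece, that specializes a parent tile class.
; It overrides CacheAllDefaultMaterials and UpdateVisualsOnClient to pass the tile flag through to the parent while forcing the grass flag false.
; The parent's grass material caching and grass mesh updates are therefore switched off for this variant.
; Its construction script is empty.
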(fn CacheAllDefaultMaterials
  (params (CacheTileMaterial :type Boolean) (CacheGrassMaterial :type Boolean))
  (|CallParentFunction CacheTileMaterial false))

(fn UpdateVisualsOnClient
  (params (UpdateTileMesh :type Boolean) (UpdateGrassMeshes :type Boolean))
  (|CallParentFunction UpdateTileMesh false))

(fn ConstructionScript)
